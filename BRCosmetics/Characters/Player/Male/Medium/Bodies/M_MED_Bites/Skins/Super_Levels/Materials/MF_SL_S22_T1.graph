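# Material function for the tier 1 Season 22 Super Level skin of the Bites character, toggled by a static switch.
# Channel masks from SL_FX select triangle, dark desaturated and fresnel glow layers, plus switchable metal and webbing layers, blended over incoming material attributes.
# Triangle effects project a scrolling texture from pre-skinned position and normal.
# Distorted, panning radial noise builds wave and hue-variation signals.
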MaterialFunction /BRCosmetics/Characters/Player/Male/Medium/Bodies/M_MED_Bites/Skins/Super_Levels/Materials/MF_SL_S22_T1
graph {nodes [
  Constant3Vector() {constant=<rgba 0 0 2 0>}
  FunctionOutput(None=StaticSwitchParameter(True=NamedRerouteUsage(), False=NamedRerouteUsage()) {group=S22_SL, parameterName=UseSL_T1, sortPriority=0})
  Multiply(A=TextureCoordinate()) {constB=0.5}
  n1 = GetMaterialAttributes(None=NamedRerouteDeclaration(None=FunctionInput() {inputName=MA, inputType=FunctionInput_MaterialAttributes}) {name=MA, nodeColor=<rgba 0 1 0.8359 1>}) {attributeGetTypes [#1, #2]}
  NamedRerouteDeclaration(None=Multiply(A=@n1.Roughness, B=ScalarParameter() {defaultValue=0.75, group=S22_SL, parameterName=RoughnessScale_T1, sliderMax=1})) {name=Roughness_In, nodeColor=<rgba 0.6172 1 0 1>}
  n2 = VectorParameter() {group=S22_SL, parameterName=TertiaryLayer_GlowColor1_T1, defaultValue=<rgba 1 0.4259 0.0576 5>}
  n3 = Multiply(A=@n2.RGB, B=@n2.RGB)
  n4 = VectorParameter() {group=S22_SL, parameterName=TertiaryLayer_GlowColor2_T1, defaultValue=<rgba 1 0.2172 0 5>}
  n5 = Multiply(A=@n4.RGB, B=@n4.RGB)
  n6 = NamedRerouteUsage()
  n7 = LinearInterpolate(B=NamedRerouteUsage(), Alpha=ScalarParameter() {defaultValue=1.5, group=S22_SL, parameterName=TertiaryLayer_AOBlend_T1, sliderMax=1})
  n8 = OneMinus(None=Fresnel() {baseReflectFraction=0, exponent=0.25})
  n9 = SetMaterialAttributes!(
   MaterialAttributes=NamedRerouteUsage()
   BaseColor=@n2.RGB
   Metallic=Constant()
   Specular
   Roughness
   EmissiveColor=Multiply(A=LinearInterpolate(A=@n3, B=@n5, Alpha=@n6), B=Multiply(A=@n7, B=@n8))
  ) {attributeSetTypes [
    #2
    #3
    #4
    #1
    #5
   ]}
  NamedRerouteDeclaration(None=@n9) {name=GlowLayer, nodeColor=<rgba 1 0.5293 0 1>}
  NamedRerouteDeclaration(None=FunctionInput() {inputName=AO, inputType=FunctionInput_Scalar, sortPriority=3}) {name=AO_In, nodeColor=<rgba 0 0.5781 1 1>}
  NamedRerouteDeclaration(None=PreSkinnedPosition()) {name=PreSkinnedLocalPosition, nodeColor=<rgba 0.0781 1 0 1>}
  n10 = FunctionInput() {inputName=SL_FX, inputType=FunctionInput_Vector4, sortPriority=1}
  n11 = ChannelMaskParameter(None=@n10) {group=S22_SL, maskChannel=Alpha, parameterName=AlphaMask, sortPriority=5, defaultValue=<rgba 0 0 0 1>}
  NamedRerouteDeclaration(None=@n11.R) {name=FX_Metal, nodeColor=<rgba 1 1 1 1>}
  n12 = Add(A=VertexInterpolator(VS=NamedRerouteUsage()).PS, B=Constant3Vector() {constant=<rgba 0 0 -125 0>})
  n13 = MaterialFunctionCall(Float2=Divide(A=ComponentMask(None=@n12) {b=true, r=true}) {constB=100}) {materialFunction=&BreakOutFloat2Components}
  n14 = Constant() {r=2}
  n15 = Constant() {r=0.2}
  n16 = Multiply(A=@n15, B=@n13.R)
  n17 = Multiply(A=@n15, B=@n13.G)
  n18 = SquareRoot(None=Add(A=Multiply(A=@n16, B=@n16), B=Multiply(A=@n17, B=@n17)))
  n19 = AppendVector(
   A=Divide(
    A=Arctangent2Fast(Y=@n13.G, X=@n13.G)
    B=MaterialFunctionCall(Multiplier=@n14) {desc=Radial Gradient, materialFunction=&Pi}.Result
   )
   B=@n18
  )
  n20 = Panner(Coordinate=Multiply(A=Constant2Vector() {g=2, r=2}, B=@n19)) {speedX=-0.01, speedY=0.01}
  n21 = Multiply(
   A=ComponentMask(None=TextureSample(UVs=@n20) {samplerType=SAMPLERTYPE_Normal, texture=/Game/Gadgets/Assets/Telephone_Booth/Textures/T_BlindsNormalNoise_N}.RGB) {g=true, r=true}
   B=Constant() {r=0.1}
  )
  n22 = Panner(Coordinate=@n19) {speedX=0.02, speedY=-0.03}
  n23 = Multiply(
   A=TextureSample(UVs=Add(A=@n21, B=@n22)) {samplerSource=SSM_Wrap_WorldGroupSettings, texture=/Game/TechArt/Seasonal/LackLuster/Textures/T_Lackluster_Phased}.RGB
   B=Power(Base=OneMinus(None=@n18))
  )
  n24 = Power(Base=@n18) {constExponent=0.7}
  NamedRerouteDeclaration(None=Multiply(A=Multiply(A=@n23, B=@n24)) {constB=0.5}) {name=Wave, nodeColor=<rgba 0 0.0859 1 1>}
  NamedRerouteDeclaration(None=@n1.BaseColor) {name=BaseColor_In, nodeColor=<rgba 0.7344 0 1 1>}
  n25 = NamedRerouteUsage()
  n26 = Multiply(
   A=Fresnel(ExponentIn=ScalarParameter() {defaultValue=5, group=S22_SL, parameterName=SecondaryLayer_FresnelExp_T1}) {baseReflectFraction=0}
   B=VectorParameter() {group=S22_SL, parameterName=SecondaryLayer_FresnelColor_T1, defaultValue=<rgba 0.2047 0.0513 0.5469 1>}.RGB
  )
  n27 = SetMaterialAttributes!(
   MaterialAttributes=@n25
   BaseColor=ScalarParameter() {defaultValue=0.5, group=S22_SL, parameterName=MetalLayer_Color_T1, sliderMax=1}
   Metallic=ScalarParameter() {defaultValue=1, group=S22_SL, parameterName=MetalLayer_Metallic_T1, sliderMax=1}
   Roughness=ScalarParameter() {defaultValue=0.3, group=S22_SL, parameterName=MetalLayer_Roughness_T1, sliderMax=1}
   EmissiveColor=@n26
  ) {attributeSetTypes [
    #2
    #3
    #1
    #5
   ]}
  NamedRerouteDeclaration(None=@n27) {name=MetalLayer, nodeColor=<rgba 0.625 0.625 0.625 1>}
  n28 = ChannelMaskParameter(None=@n10) {group=S22_SL, maskChannel=Alpha, parameterName=WebbingMask, sortPriority=5, defaultValue=<rgba 0 0 0 1>}
  NamedRerouteDeclaration(None=@n28.R) {name=FX_Webbing, nodeColor=<rgba 0.2842 0 1 1>}
  n29 = VectorParameter() {group=S22_SL, parameterName=WebbingLayer_GlowColor_T1, defaultValue=<rgba 1 0 0.79 1>}
  n30 = Multiply(A=@n29.RGB, B=@n29.RGB)
  n31 = OneMinus(None=Fresnel() {baseReflectFraction=0, exponent=1})
  NamedRerouteDeclaration(None=SetMaterialAttributes!(MaterialAttributes=NamedRerouteUsage(), BaseColor=@n29.RGB, Metallic=Constant(), EmissiveColor=Multiply(A=@n30, B=@n31)) {attributeSetTypes [#2, #3, #5]}) {name=WebbingLayer, nodeColor=<rgba 0.4841 0 1 1>}
  NamedRerouteDeclaration(None=ChannelMaskParameter(None=@n10) {group=S22_SL, parameterName=PrimaryMask, sortPriority=5}.R) {name=FX_Triangle, nodeColor=<rgba 1 0 0.6094 1>}
  n32 = Add(A=@n22) {constB=0.01}
  n33 = MaterialFunctionCall(Float3=TextureSample(UVs=Add(A=@n21, B=@n32)) {samplerSource=SSM_Wrap_WorldGroupSettings, texture=&T_Lackluster_Phased}.RGB) {materialFunction=&BreakOutFloat3Components}
  n34 = Multiply(A=Add(A=Add(A=@n33.R, B=@n33.R), B=@n33.B)) {constB=0.333}
  n35 = OneMinus(None=Power(Base=OneMinus(None=@n34)) {constExponent=10})
  n36 = ScalarParameter() {defaultValue=0.3, group=S22_SL, parameterName=HueShift_Intensity_T1, sliderMax=1}
  NamedRerouteDeclaration(None=Multiply(A=Multiply(A=@n35, B=@n36) {constB=0.3}) {constB=-1}) {name=VariationCurve, nodeColor=<rgba 1 0.9375 0 1>}
  NamedRerouteDeclaration(None=VertexInterpolator(VS=PreSkinnedNormal()).PS) {name=PreSkinnedLocalNormal_VI, nodeColor=<rgba 0 0.3672 1 1>}
  n37 = ChannelMaskParameter(None=@n10) {group=S22_SL, maskChannel=Green, parameterName=SecondaryMask, sortPriority=5, defaultValue=<rgba 0 1 0 0>}
  NamedRerouteDeclaration(None=@n37.R) {name=FX_Black, nodeColor=<rgba 0.0208 0.0208 0.0208 1>}
  n38 = ChannelMaskParameter(None=@n10) {group=S22_SL, maskChannel=Blue, parameterName=TertiaryMask, sortPriority=5, defaultValue=<rgba 0 0 1 0>}
  NamedRerouteDeclaration(None=@n38.R) {name=FX_Glow, nodeColor=<rgba 1 0.2846 0 1>}
  n39 = NamedRerouteUsage()
  n40 = NamedRerouteUsage()
  n41 = NamedRerouteUsage()
  n42 = NamedRerouteUsage()
  n43 = NamedRerouteUsage()
  n44 = BlendMaterialAttributes(
   A=BlendMaterialAttributes(A=BlendMaterialAttributes(A=@n41, B=@n42, Alpha=@n43), B=NamedRerouteUsage(), Alpha=NamedRerouteUsage())
   B=NamedRerouteUsage()
   Alpha=NamedRerouteUsage()
  )
  n45 = StaticSwitchParameter(True=BlendMaterialAttributes(A=@n44, B=NamedRerouteUsage(), Alpha=NamedRerouteUsage()), False=@n44) {group=S22_SL, parameterName=UseMetalChannel, sortPriority=1}
  NamedRerouteDeclaration(None=StaticSwitchParameter(True=BlendMaterialAttributes(A=@n45, B=@n39, Alpha=@n40), False=@n45) {group=S22_SL, parameterName=UseRoseDustWebbing, sortPriority=1}) {name=MA_Out, nodeColor=<rgba 0.8634 1 0 1>}
  n46 = ScalarParameter() {defaultValue=2, group=S22_SL, parameterName=PrimaryLayer_FresnelExp_T1, sliderMax=8}
  n47 = TextureObject() {samplerType=SAMPLERTYPE_LinearColor, texture=/Game/TechArt/Seasonal/LackLuster/Textures/T_Lackluster_Triangles}
  n48 = Constant() {r=75}
  n49 = Time()
  n50 = Constant3Vector() {constant=<rgba -0.5 0.5 -1 0>}
  n51 = VertexInterpolator(VS=Add(A=NamedRerouteUsage(), B=Multiply(A=@n49, B=@n50)))
  n52 = NamedRerouteUsage()
  n53 = Constant() {r=0.9}
  n54 = OneMinus(None=ComponentMask(None=MaterialFunctionCall(TextureObject=@n47, TextureSize=@n48, WorldPosition=@n51.PS, World Space Normal=@n52, ProjectionTransitionContrast=@n53) {materialFunction=&WorldAlignedTexture}.XY Texture) {g=true})
  n55 = Multiply(A=Fresnel(ExponentIn=ScalarParameter() {defaultValue=1.5, group=S22_SL, parameterName=PrimaryLayer_Triangle_FresnelExp_T1, sliderMax=8}) {baseReflectFraction=0, exponent=4}) {constB=2}
  n56 = Multiply(
   A=Add(A=Fresnel(ExponentIn=@n46) {baseReflectFraction=0, exponent=4}, B=Multiply(A=@n54, B=@n55))
   B=MaterialFunctionCall(
    Hue Shift Percentage=NamedRerouteUsage()
    Texture=VectorParameter() {group=S22_SL, parameterName=PrimaryLayer_FresnelColor_T1, defaultValue=<rgba 0.2047 0.0513 0.5469 1>}.RGB
   ) {materialFunction=&HueShift}.Result
  )
  n57 = SetMaterialAttributes!(
   MaterialAttributes=NamedRerouteUsage()
   BaseColor=Constant() {r=0.04}
   Metallic=Constant() {r=1}
   Specular
   Roughness=NamedRerouteUsage()
   EmissiveColor=Add(A=NamedRerouteDeclaration(None=@n56) {name=TriangleFX, nodeColor=<rgba 0.3828 0 1 1>}, B=NamedRerouteUsage())
  ) {attributeSetTypes [
    #2
    #3
    #4
    #1
    #5
   ]}
  NamedRerouteDeclaration(None=@n57) {name=TrinangleLayer, nodeColor=<rgba 1 0.0369 0.8579 1>}
  n58 = Multiply(
   A=Desaturation(None=NamedRerouteUsage(), Fraction=Constant() {r=1})
   B=ScalarParameter() {defaultValue=1.75, group=S22_SL, parameterName=SecondaryLayer_DesatColorScale_T1, sliderMax=20}
  ) {constB=10}
  n59 = NamedRerouteUsage()
  n60 = SetMaterialAttributes(
   MaterialAttributes=@n25
   BaseColor=LinearInterpolate(
    A=ScalarParameter() {defaultValue=0.04, group=S22_SL, parameterName=SecondaryLayer_Color_T1, sliderMax=1}
    B=Saturate(None=@n58)
    Alpha=ScalarParameter() {defaultValue=1, group=S22_SL, parameterName=SecondaryLayer_DesatColorBlend_T1, sliderMax=1}
   )
   Metallic=ScalarParameter() {defaultValue=1, group=S22_SL, parameterName=SecondaryLayer_Metallic_T1, sliderMax=1}
   Roughness=NamedRerouteUsage()
   EmissiveColor=LinearInterpolate(A=@n26, B=Add(A=@n26, B=@n59), Alpha=ScalarParameter() {group=S22_SL, parameterName=SecondaryLayer_BlendPrimaryFX_T1, sliderMax=1})
  ) {attributeSetTypes [
    #2
    #3
    #4
    #1
    #5
   ]}
  NamedRerouteDeclaration(None=@n60) {name=BlackLayer, nodeColor=<rgba 0.1719 0.1719 0.1719 1>}
 ]}
