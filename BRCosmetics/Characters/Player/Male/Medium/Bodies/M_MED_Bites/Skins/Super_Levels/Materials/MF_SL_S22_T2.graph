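# Material function for the Season 22 Super Levels tier 2 skin style on M_MED_Bites, applied to incoming material attributes behind the UseSL_T2 switch.
# SL_FX channel masks blend iridescent, desaturated secondary, glow and optional RoseDust webbing layers.
# The iridescent layer combines panning thin-film Fresnel, hue shift, triangle sparkle, polar-coordinate waves and animated foil normals projected from pre-skinned position.
# Shading-path switches simplify foil normals and distortion on mobile.
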MaterialFunction /BRCosmetics/Characters/Player/Male/Medium/Bodies/M_MED_Bites/Skins/Super_Levels/Materials/MF_SL_S22_T2
graph {nodes [
  FunctionOutput(None=StaticSwitchParameter(True=NamedRerouteUsage(), False=NamedRerouteUsage()) {group=S22_SL, parameterName=UseSL_T2, sortPriority=0})
  Multiply(A=NamedRerouteUsage(), B=ScalarParameter() {defaultValue=0.75, group=S22_SL, parameterName=MetalLayer_RoughnessScale_T2, sliderMax=1})
  Multiply(A=NamedRerouteUsage(), B=ScalarParameter() {defaultValue=1.35, group=S22_SL, parameterName=GlowLayer_RoughnessScale_T2, sliderMax=1})
  n1 = VectorParameter() {group=S22_SL, parameterName=GlowLayer_Emissive_T2}
  n2 = SetMaterialAttributes!(
   MaterialAttributes=NamedRerouteUsage()
   BaseColor=VectorParameter() {group=S22_SL, parameterName=GlowLayer_BaseColor_T2, defaultValue=<rgba 1 1 1 0>}.RGB
   Metallic=ScalarParameter() {defaultValue=1, group=S22_SL, parameterName=GlowLayer_Metallic}
   Specular
   Roughness
   EmissiveColor=Multiply(A=@n1.RGB, B=@n1.RGB)
  ) {attributeSetTypes [
    #1
    #2
    #3
    #4
    #5
   ]}
  NamedRerouteDeclaration(None=@n2) {name=WhiteLayer2, nodeColor=<rgba 0.9479 0.9539 1 1>}
  n3 = Multiply(
   A=Desaturation(None=NamedRerouteUsage(), Fraction=Constant() {r=1})
   B=ScalarParameter() {defaultValue=5, group=S22_SL, parameterName=SecondaryLayer_DesatColorScale_T2, sliderMax=20}
  ) {constB=10}
  n4 = SetMaterialAttributes(
   MaterialAttributes=NamedRerouteUsage()
   BaseColor=LinearInterpolate(
    A=ScalarParameter() {defaultValue=0.9, group=S22_SL, parameterName=SecondaryLayer_BaseColor_T2, sliderMax=1}
    B=Saturate(None=@n3)
    Alpha=ScalarParameter() {defaultValue=0.5, group=S22_SL, parameterName=SecondaryLayer_DesatColorBlend_T2, sliderMax=1}
   )
   Metallic=ScalarParameter() {defaultValue=1, group=S22_SL, parameterName=SecondaryLayer_Metallic_T2, sliderMax=1}
   Roughness=Multiply(A=NamedRerouteUsage(), B=ScalarParameter() {defaultValue=1, group=S22_SL, parameterName=SecondaryLayer_RoughnessScale_T2, sliderMax=1}) {constB=2}
  ) {attributeSetTypes [
    #1
    #2
    #3
    #4
    #5
   ]}
  NamedRerouteDeclaration(None=@n4) {name=WhiteLayer, nodeColor=<rgba 0.9479 0.9539 1 1>}
  NamedRerouteDeclaration(None=FunctionInput() {inputName=AO, inputType=FunctionInput_Scalar, sortPriority=3}) {name=AO_In, nodeColor=<rgba 0 0.5781 1 1>}
  NamedRerouteDeclaration(None=FunctionInput(Preview=PreSkinnedPosition()) {bUsePreviewValueAsDefault=true, inputName=PreSkinnedLocalPosition_VS, sortPriority=4}) {name=PreSkinnedLocalPosition_VS, nodeColor=<rgba 0.0781 1 0 1>}
  n5 = TextureObject() {samplerType=SAMPLERTYPE_Normal, texture=/Game/TechArt/Seasonal/LackLuster/Textures/T_Lackluster_Swirl_N}
  n6 = NamedRerouteUsage()
  n7 = Constant() {r=0.9}
  n8 = Constant() {r=200}
  n9 = Time()
  n10 = Constant3Vector() {constant=<rgba 0.5 -0.5 1 0>}
  n11 = VertexInterpolator(VS=Add(A=NamedRerouteUsage(), B=Multiply(A=@n9, B=@n10)))
  NamedRerouteDeclaration(None=ShadingPathSwitch(
    Default=MaterialFunctionCall!(TextureObject=@n5, NormalWS=@n6, ProjectionTransitionSharpness=@n7, Size=@n8, NormalFlatness, Position=@n11.PS) {materialFunction=/Game/Global/MaterialFunctions/MF_WorldAlignedNormal_Whiteout}.XYZ Texture
    Mobile=Constant3Vector() {constant=<rgba 0 0 1 0>}
   )) {name=FoilNormals, nodeColor=<rgba 1 0.6797 0 1>}
  n12 = GetMaterialAttributes(None=NamedRerouteDeclaration(None=FunctionInput() {inputName=MA, inputType=FunctionInput_MaterialAttributes}) {name=MA, nodeColor=<rgba 0 1 0.8359 1>}) {attributeGetTypes [#4, #6, #1]}
  NamedRerouteDeclaration(None=@n12.Normal) {name=Normal_In, nodeColor=<rgba 0 1 0.6016 1>}
  NamedRerouteDeclaration(None=@n12.BaseColor) {name=BaseColor_In, nodeColor=<rgba 0.7344 0 1 1>}
  NamedRerouteDeclaration(None=@n12.Roughness) {name=Roughness_In, nodeColor=<rgba 0 0.3203 1 1>}
  n13 = FunctionInput() {inputName=SL_FX, inputType=FunctionInput_Vector4, sortPriority=1}
  n14 = ChannelMaskParameter(None=@n13) {group=S22_SL, maskChannel=Alpha, parameterName=WebbingMask, sortPriority=5, defaultValue=<rgba 0 0 0 1>}
  NamedRerouteDeclaration(None=@n14.R) {name=FX_Webbing, nodeColor=<rgba 0.2842 0 1 1>}
  n15 = VectorParameter() {group=S22_SL, parameterName=WebbingLayer_GlowColor_T, defaultValue=<rgba 1 0 0.79 1>}
  n16 = Fresnel() {baseReflectFraction=0, exponent=1}
  n17 = SetMaterialAttributes!(
   MaterialAttributes=NamedRerouteUsage()
   BaseColor=@n15.RGB
   Metallic=Constant()
   EmissiveColor=Multiply(A=Multiply(A=@n15.RGB, B=@n15.RGB), B=OneMinus(None=@n16))
   Roughness
  ) {attributeSetTypes [
    #1
    #2
    #5
    #4
   ]}
  NamedRerouteDeclaration(None=@n17) {name=WebbingLayer, nodeColor=<rgba 0.4841 0 1 1>}
  NamedRerouteDeclaration(None=ChannelMaskParameter(None=@n13) {group=S22_SL, parameterName=PrimaryMask, sortPriority=5}.R) {name=FX_Iridescent, nodeColor=<rgba 0 0.1134 1 1>}
  n18 = Constant2Vector() {g=2, r=2}
  n19 = Add(A=VertexInterpolator(VS=NamedRerouteUsage()).PS, B=Constant3Vector() {constant=<rgba 0 0 -125 0>})
  n20 = MaterialFunctionCall(Float2=Divide(A=ComponentMask(None=@n19) {b=true, r=true}) {constB=100}) {materialFunction=&BreakOutFloat2Components}
  n21 = Constant() {r=2}
  n22 = Constant() {r=0.2}
  n23 = Multiply(A=@n22, B=@n20.R)
  n24 = Multiply(A=@n22, B=@n20.G)
  n25 = SquareRoot(None=Add(A=Multiply(A=@n23, B=@n23), B=Multiply(A=@n24, B=@n24)))
  n26 = AppendVector(
   A=Divide(
    A=Arctangent2Fast(Y=@n20.G, X=@n20.G)
    B=MaterialFunctionCall(Multiplier=@n21) {desc=Radial Gradient, materialFunction=&Pi}.Result
   )
   B=@n25
  ) {desc=PolarCoords}
  n27 = TextureSample(UVs=Panner(Coordinate=Multiply(A=@n18, B=@n26)) {speedX=-0.01, speedY=0.01}) {samplerType=SAMPLERTYPE_Normal, texture=/Game/Gadgets/Assets/Telephone_Booth/Textures/T_BlindsNormalNoise_N}
  n28 = ShadingPathSwitch(Default=Multiply(A=ComponentMask(None=@n27.RGB) {g=true, r=true}, B=Constant() {r=0.1}), Mobile=Constant())
  n29 = Panner(Coordinate=@n26) {speedX=0.02, speedY=-0.03}
  n30 = Add(A=@n28, B=Add(A=@n29) {constB=0.01})
  n31 = TextureSample(UVs=Add(A=@n28, B=@n29)) {samplerSource=SSM_Wrap_WorldGroupSettings, texture=/Game/TechArt/Seasonal/LackLuster/Textures/T_Lackluster_Phased}
  n32 = MaterialFunctionCall(Float3=ShadingPathSwitch(
    Default=TextureSample(UVs=@n30) {samplerSource=SSM_Wrap_WorldGroupSettings, texture=&T_Lackluster_Phased}.RGB
    Mobile=@n31.RGB
   )) {materialFunction=&BreakOutFloat3Components}
  n33 = Add(A=Add(A=@n32.R, B=@n32.R), B=@n32.B)
  n34 = Power(Base=OneMinus(None=Multiply(A=@n33) {constB=0.333})) {constExponent=10}
  NamedRerouteDeclaration(None=Multiply(A=OneMinus(None=@n34), B=ScalarParameter() {defaultValue=0.12, group=S22_SL, parameterName=HueShift_Intensity_T2, sliderMax=1}) {constB=0.12}) {name=VariationCurve, nodeColor=<rgba 1 0.9375 0 1>}
  n35 = Multiply(A=@n31.RGB, B=Power(Base=OneMinus(None=@n25)))
  n36 = Power(Base=@n25) {constExponent=0.7}
  NamedRerouteDeclaration(None=Multiply(A=Multiply(A=@n35, B=@n36)) {constB=0.5}) {name=Wave, nodeColor=<rgba 0 0.0859 1 1>}
  n37 = ChannelMaskParameter(None=@n13) {group=S22_SL, maskChannel=Green, parameterName=SecondaryMask, sortPriority=5, defaultValue=<rgba 0 1 0 0>}
  NamedRerouteDeclaration(None=@n37.R) {name=FX_White, nodeColor=<rgba 1 1 1 1>}
  n38 = ChannelMaskParameter(None=@n13) {group=S22_SL, maskChannel=Blue, parameterName=TertiaryMask, sortPriority=5, defaultValue=<rgba 0 0 1 0>}
  NamedRerouteDeclaration(None=@n38.R) {name=FX_White2?, nodeColor=<rgba 1 0.9283 0.8553 1>}
  n39 = NamedRerouteUsage()
  n40 = NamedRerouteUsage()
  n41 = NamedRerouteUsage()
  n42 = NamedRerouteUsage()
  n43 = NamedRerouteUsage()
  n44 = BlendMaterialAttributes(
   A=BlendMaterialAttributes(A=BlendMaterialAttributes(A=@n41, B=@n42, Alpha=@n43), B=NamedRerouteUsage(), Alpha=NamedRerouteUsage())
   B=NamedRerouteUsage()
   Alpha=NamedRerouteUsage()
  )
  NamedRerouteDeclaration(None=StaticSwitchParameter(True=BlendMaterialAttributes(A=@n44, B=@n39, Alpha=@n40), False=@n44) {group=S22_SL, parameterName=UseRoseDustWebbing, sortPriority=1}) {name=MA_Out, nodeColor=<rgba 0.8634 1 0 1>}
  n45 = ScalarParameter() {defaultValue=1, group=S22_SL, parameterName=PrimaryLayer_FresnelExp_T2, sliderMax=8}
  n46 = TextureObject() {samplerType=SAMPLERTYPE_LinearColor, texture=/Game/TechArt/Seasonal/LackLuster/Textures/T_Lackluster_Triangles}
  n47 = Constant() {r=75}
  n48 = Time()
  n49 = Constant3Vector() {constant=<rgba -0.5 0.5 -1 0>}
  n50 = VertexInterpolator(VS=Add(A=NamedRerouteUsage(), B=Multiply(A=@n48, B=@n49)))
  n51 = NamedRerouteUsage()
  n52 = Constant() {r=0.9}
  n53 = OneMinus(None=ComponentMask(None=MaterialFunctionCall(TextureObject=@n46, TextureSize=@n47, WorldPosition=@n50.PS, World Space Normal=@n51, ProjectionTransitionContrast=@n52) {materialFunction=&WorldAlignedTexture}.XY Texture) {g=true})
  n54 = Multiply(A=Fresnel(ExponentIn=ScalarParameter() {defaultValue=1, group=S22_SL, parameterName=PrimaryLayer_Triangle_FresnelExp_T2, sliderMax=8}) {baseReflectFraction=0, exponent=4}) {constB=3}
  n55 = Multiply(
   A=Add(A=Fresnel(ExponentIn=@n45) {baseReflectFraction=0, exponent=4}, B=Multiply(A=@n53, B=@n54))
   B=MaterialFunctionCall(
    Hue Shift Percentage=NamedRerouteUsage()
    Texture=VectorParameter() {group=S22_SL, parameterName=PrimaryLayer_FresnelColor_T2, defaultValue=<rgba 0 0.2281 0.7031 1>}.RGB
   ) {materialFunction=&HueShift}.Result
  )
  n56 = Multiply(A=Time()) {constB=-0.03}
  n57 = ScalarParameter() {defaultValue=0.5, group=S22_SL, parameterName=PrimaryLayer_Iridescent_FresnelExpo, sliderMax=8}
  n58 = Transform(None=NamedRerouteUsage())
  n59 = MaterialFunctionCall(X=Add(A=Frac(None=@n56), B=Fresnel(ExponentIn=@n57, Normal=@n58) {baseReflectFraction=0, exponent=1}), Y=Constant()) {materialFunction=&MakeFloat2}
  n60 = Multiply(
   A=Multiply(A=TextureSample(UVs=@n59.Result) {texture=/Game/Characters/Player/Male/Medium/Bodies/M_MED_Bites/Skins/Super_Levels/Materials/T_ThinFilm_S22_SL_T2}.RGB) {constB=0.2}
   B=Fresnel(BaseReflectFractionIn=ScalarParameter() {group=S22_SL, parameterName=PrimaryLayer_Iridescent_BaseReflect, sliderMax=1}) {baseReflectFraction=0.5, exponent=1}
  )
  n61 = NamedRerouteUsage()
  n62 = Constant() {r=0.65}
  n63 = SetMaterialAttributes!(
   MaterialAttributes=NamedRerouteUsage()
   BaseColor=VectorParameter() {group=S22_SL, parameterName=PrimaryLayer_BaseColor_T2, defaultValue=<rgba 0.0141 0.1023 0.2344 1>}.RGB
   Metallic=Constant() {r=1}
   Specular
   Roughness=Multiply(A=NamedRerouteUsage(), B=ScalarParameter() {defaultValue=0.7, group=S22_SL, parameterName=PrimaryLayer_RoughnessScale_T2, sliderMax=1})
   EmissiveColor=Add(A=Add(A=@n55, B=@n60), B=NamedRerouteUsage()) {constB=0}
   Normal=MaterialFunctionCall(
    BaseNormal=NamedRerouteUsage()
    AdditionalNormal=MaterialFunctionCall(Normal=@n61, Flatness=@n62) {materialFunction=&FlattenNormal}.Result
   ) {materialFunction=&BlendAngleCorrectedNormals}.Result
  ) {attributeSetTypes [
    #1
    #2
    #3
    #4
    #5
    #6
   ]}
  NamedRerouteDeclaration(None=@n63) {name=IridescentLayer, nodeColor=<rgba 0 0.6191 1 1>}
  n64 = PreSkinnedNormal()
  NamedRerouteDeclaration(None=FunctionInput(Preview=VertexInterpolator(VS=@n64).PS) {bUsePreviewValueAsDefault=true, inputName=PreSkinnedLocalNormal_PS, sortPriority=5}) {name=PreSkinnedLocalNormal_PS, nodeColor=<rgba 0 0.3672 1 1>}
 ]}
