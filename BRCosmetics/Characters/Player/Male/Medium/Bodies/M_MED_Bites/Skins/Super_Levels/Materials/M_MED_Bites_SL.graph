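# Master material for the Bites character body's Super Levels skin, using material attributes with Default Lit and Subsurface Profile shading.
# The base character PreFX function feeds a static switch, on by default, that applies the Bites pickaxe's liquid metal function masked by an FX mask channel.
# Three chained Season 22 Super Level tier functions share a packed FX texture and AO, then feed the character PostFX and cinematic PostFX functions.
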Material /BRCosmetics/Characters/Player/Male/Medium/Bodies/M_MED_Bites/Skins/Super_Levels/Materials/M_MED_Bites_SL
graph {
 properties {
  bAutomaticallySetUsageInEditor=false
  bUseMaterialAttributes=true
  bUsedWithMeshDeformer=true
  bUsedWithMorphTargets=true
  bUsedWithSkeletalMesh=true
  pixelDepthOffsetMode=PDOM_Legacy
  shadingModel=MSM_FromMaterialExpression
  subsurfaceProfile=&SS_HeroSkin_02
  usedShadingModels=Default Lit | Subsurface Profile
 }
 nodes [
  n1 = MaterialFunctionCall() {materialFunction=&BaseFN_CharacterFunction_PreFX}
  n2 = StaticComponentMaskParameter(None=TextureSampleParameter2D() {parameterName=FX_Mask, samplerType=SAMPLERTYPE_Masks, texture=&T_BlackMask}.RGB) {defaultG=true, parameterName=FXChannel}
  n3 = TextureSampleParameter2D() {
   group=S22_SL
   parameterName=SL_FX
   samplerType=SAMPLERTYPE_Masks
   sortPriority=1
   texture=/BRCosmetics/Characters/Player/Male/Medium/Bodies/M_MED_Bites/Skins/Super_Levels/Textures/T_Character_Bites_Body_SL_FX
  }
  n4 = MaterialFunctionCall!(
   MA=StaticSwitchParameter(
    True=MaterialFunctionCall!(material=@n1.Result, AO=@n1.Result, PS_Position=@n1.Result, FxMask=@n2) {materialFunction=/BRCosmetics/Weapons/FORT_Melee/Pickaxe_Bites/Material/MF_LiquidMetal}.material
    False=@n1.Result
   ) {defaultValue=true, group=1-Bites, parameterName=UseLiquidMetal}
   SL_FX=@n3.RGBA
   AO=@n1.AO
  ) {materialFunction=&/MF_SL_S22_T1}
  n5 = MaterialFunctionCall(
   AO=@n1.AO
   Distance Blend=@n1.AO
   Fresnel=@n1.AO
   Gmap=@n1.AO
   Pre Skinned Local Position=@n1.AO
   Material=MaterialFunctionCall!(
    MA=MaterialFunctionCall!(MA=@n4.Result, SL_FX=@n3.RGBA, AO=@n1.AO, PreSkinnedLocalPosition_VS, PreSkinnedLocalNormal_PS) {materialFunction=&/MF_SL_S22_T2}.Result
    SL_FX=@n3.RGBA
    AO=@n1.AO
   ) {materialFunction=&/MF_SL_S22_T3}.Result
  ) {materialFunction=&BaseFN_CharacterFunction_PostFX}
  MaterialFunctionCall(In=@n5.Result) {materialFunction=&CINE_MF_CharPostFX}
 ]
}
